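; Pumpkin decor actor (variant 3_B) among the VoiceFork BR quest actors, which ties its lifetime to a linked actor.
; It handles FortLinkToActor events, printing a debug message when the link connects.
; When the linked actor is destroyed, it prints a debug message and destroys itself.
; BeginPlay, Tick, overlap and the construction script are empty.
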(event EventBeginPlay)

(event Collision|EventActorBeginOverlap
  (params (OtherActor :type Actor%20Object%20Reference)))

(event EventTick
  (params (DeltaSeconds :type Float%20%28single-precision%29)))

(event OnLinkedActorDestroyed(FortLinkToActor)
  (params (DamageCauser :type Actor%20Object%20Reference))
  (Development|PrintString "Pumpkin Destroyed" true true "(R=1.000000,G=0.000000,B=0.000000,A=1.000000)" 5.0)
  (Actor|DestroyActor))

(event OnLinkedActorConnected(FortLinkToActor)
  (params (LinkedActor :type Actor%20Object%20Reference))
  (Development|PrintString "Pumpkin connected" true true "(R=1.000000,G=0.000000,B=0.000000,A=1.000000)" 5.0))

(fn ConstructionScript)
